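# Translucent unlit particle material for a soft round glow, stored with the Meteorwomen Alt backpack FX and enabled for Niagara sprites and mesh particles.
# A UV-centered SphereMask with adjustable radius, hardness and power forms the procedural glow shape without textures.
# The shape multiplies particle color into MF_FX_EmissiveAdjustments, and particle alpha for transparency.
# A shading path switch applies depth fade to that alpha by default and skips it on mobile.
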Material /Game/Accessories/FORT_Backpacks/Backpack_F_MED_Meteorwomen_Alt/FX/M_FX_Generic_Soft_Round_Glow
graph {
 properties {
  bUsedWithNanite=true
  bUsedWithNiagaraMeshParticles=true
  bUsedWithNiagaraSprites=true
  blendMode=BLEND_Translucent
  pixelDepthOffsetMode=PDOM_Legacy
  shadingModel=MSM_Unlit
  usedShadingModels=Unlit
 }
 nodes [
  n1 = ParticleColor()
  n2 = Power(
   Base=SphereMask(
    A=TextureCoordinate()
    B=Constant2Vector() {g=0.5, r=0.5}
    Radius=ScalarParameter() {defaultValue=0.5, parameterName=Radius}
    Hardness=ScalarParameter() {parameterName=Hardness}
   )
   Exp=ScalarParameter() {defaultValue=2, parameterName=Sphere_Power_Exp}
  )
  MaterialFunctionCall(Emissive=Multiply(A=@n1.RGB, B=@n2)) {materialFunction=&MF_FX_EmissiveAdjustments}
  n3 = Multiply(A=@n2, B=@n1.A)
  n4 = ScalarParameter() {defaultValue=50, parameterName=FadeDistance}
  ShadingPathSwitch(Default=Multiply(A=@n3, B=DepthFade(FadeDistance=@n4)), Mobile=@n3) {desc=Migrated from FeatureLevelSwitch}
 ]
}
